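# Material function in the ChicleVeil pickaxe folder that converts UVs into tiled, offsettable radial (polar) coordinates.
# Distance from the 0.5 center gives U, and a fast Atan2 angle normalized by two pi gives V.
# Both are wrapped with Frac, scaled by tiling, wrapped again, then added to a panner offset.
# Separate DDX and DDY outputs expose derivatives, useful for gradient-based sampling across the wrap seams.
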MaterialFunction /BRCosmetics/Weapons/FORT_Melee/Pickaxe_ChicleVeil/Materials/MF_RadialUVPanner
graph {
 properties {libraryCategoriesText [Misc, ""]}
 nodes [
  n1 = Distance(A=NamedRerouteUsage(), B=Constant() {r=0.5})
  n2 = NamedRerouteUsage()
  n3 = Frac(None=Multiply(A=Frac(None=@n1), B=ComponentMask(None=@n2) {r=true}) {constB=5})
  n4 = ConstantBiasScale(None=NamedRerouteUsage()) {bias=-0.5, scale=2}
  n5 = Divide(
   A=Arctangent2Fast(Y=ComponentMask(None=@n4) {r=true}, X=ComponentMask(None=@n4) {g=true})
   B=MaterialFunctionCall(Multiplier=Constant() {r=2}) {materialFunction=&Pi}.Result
  )
  n6 = NamedRerouteUsage()
  n7 = Frac(None=Multiply(A=Frac(None=@n5), B=ComponentMask(None=@n6) {g=true}) {constB=2})
  FunctionOutput(None=Add(A=NamedRerouteUsage(), B=AppendVector(A=@n3, B=@n7) {desc=Final UV}) {constB=-0.5}) {outputName=UV}
  n8 = NamedRerouteUsage()
  FunctionOutput(None=DDX(Value=@n8)) {outputName=DDX, sortPriority=1}
  FunctionOutput(None=DDY(Value=@n8)) {outputName=DDY, sortPriority=2}
  n9 = FunctionInput() {bUsePreviewValueAsDefault=true, inputName=UV Tilling, inputType=FunctionInput_Vector2, sortPriority=1, previewValue=<1 1 0 1>}
  NamedRerouteDeclaration(None=@n9) {name=UVTilling, nodeColor=<rgba 1 0 0.0703 1>}
  NamedRerouteDeclaration(None=FunctionInput(Preview=TextureCoordinate()) {inputName=UV, inputType=FunctionInput_Vector2}) {name=UV, nodeColor=<rgba 1 0 0.4453 1>}
  NamedRerouteDeclaration(None=FunctionInput() {bUsePreviewValueAsDefault=true, inputName=PannerValue, inputType=FunctionInput_Vector2, sortPriority=2}) {name=Offset, nodeColor=<rgba 1 0 0.6562 1>}
 ]
}
