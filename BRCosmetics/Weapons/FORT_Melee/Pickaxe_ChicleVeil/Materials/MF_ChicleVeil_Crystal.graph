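# Material function for the ChicleVeil pickaxe crystal, blending a crystal look over the input material attributes by Alpha.
# A bump-offset interior texture, flattened on lower quality levels, tints the base color, alongside fresnel and sky-light reflection terms for emissive.
# Optional spell FX radially pans a borrowed DyedDuelist texture over pivot UVs built from pre-skinned position.
# Reactive, ColorProgress and SpellSpeed inputs feed the color shifts and animation.
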MaterialFunction /BRCosmetics/Weapons/FORT_Melee/Pickaxe_ChicleVeil/Materials/MF_ChicleVeil_Crystal
graph {nodes [
  ComponentMask() {g=true, r=true}
  n1 = NamedRerouteUsage()
  n2 = NamedRerouteUsage()
  n3 = NamedRerouteUsage()
  n4 = NamedRerouteUsage()
  n5 = VectorParameter() {group=CVCrystal, parameterName=CV_ReactiveCrystalColor, sortPriority=29, defaultValue=<rgba 0.1668 0.0575 0.276 1>}
  n6 = LinearInterpolate(
   A=Multiply(
    A=VectorParameter() {group=CVCrystal, parameterName=CV_BaseColorMultiply, sortPriority=0, defaultValue=<rgba 1 1 1 1>}.RGB
    B=NamedRerouteUsage()
   )
   B=@n5.RGB
   Alpha=Multiply(A=NamedRerouteUsage(), B=NamedRerouteUsage())
  ) {constAlpha=0}
  n7 = TextureSampleParameter2D(UVs=NamedRerouteUsage()) {
   group=CVCrystal
   parameterName=CV_InteriorTexture
   sortPriority=5
   texture=/BRCosmetics/Weapons/FORT_Melee/Pickaxe_ChicleVeil/Textures/T_ChicleVeil_Crystal
  }
  n8 = Constant() {r=0.5}
  n9 = Desaturation(
   None=QualitySwitch(Default=@n7.RGB, Low=@n8, High=@n7.RGB, Medium=@n8, Epic=@n7.RGB)
   Fraction=OneMinus(None=ScalarParameter() {defaultValue=0.5, group=CVCrystal, parameterName=CV_InteriorPrismColors, sortPriority=2})
  )
  n10 = LinearInterpolate(
   A=@n6
   B=Multiply(A=Multiply(A=@n6, B=@n9), B=ScalarParameter() {defaultValue=12, group=CVCrystal, parameterName=CV_InteriorIntensity, sortPriority=1}) {constB=0.1}
   Alpha=ScalarParameter() {defaultValue=0.6, group=CVCrystal, parameterName=CV_InteriorContrast, sortPriority=3}
  )
  n11 = Multiply(A=Desaturation(None=NamedRerouteUsage()), B=Multiply(A=NamedRerouteUsage(), B=NamedRerouteUsage()))
  n12 = NamedRerouteDeclaration(None=StaticSwitchParameter(True=LinearInterpolate(A=@n10, Alpha=@n11) {constAlpha=0, constB=0}, False=@n10) {group=CVCrystal, parameterName=CV_hasSpellFX, sortPriority=30}) {name=InteriorColor, nodeColor=<rgba 1 0 0.4688 1>}
  n13 = NamedRerouteUsage()
  n14 = Add(A=@n13, B=NamedRerouteUsage())
  n15 = Multiply(A=NamedRerouteUsage(), B=NamedRerouteUsage())
  n16 = Add(
   A=StaticSwitchParameter(True=LinearInterpolate(A=@n13, B=@n14, Alpha=@n15) {constAlpha=0}, False=NamedRerouteUsage()) {group=CVCrystal, parameterName=CV_hasSpellFX, sortPriority=30}
   B=NamedRerouteUsage()
  )
  n17 = VectorParameter() {group=CVCrystal, parameterName=CV_ReactiveCrystalColorEmissive, sortPriority=29, defaultValue=<rgba 0.2616 0.1202 0.2865 1>}
  n18 = NamedRerouteUsage()
  n19 = NamedRerouteUsage()
  n20 = Multiply(
   A=LinearInterpolate(A=@n16, B=Add(A=@n16, B=@n17.RGB), Alpha=Multiply(A=@n18, B=@n19)) {constAlpha=0}
   B=ScalarParameter() {defaultValue=0.5, parameterName=CV_EmissiveIntensity}
  )
  n21 = Constant()
  n22 = FunctionInput() {bUsePreviewValueAsDefault=true, inputName=Alpha, inputType=FunctionInput_Scalar, sortPriority=1, previewValue=<1 1 1 1>}
  FunctionOutput(None=BlendMaterialAttributes(
    A=SetMaterialAttributes!(MaterialAttributes=@n1, BaseColor=@n2, EmissiveColor=@n3) {attributeSetTypes [#1, #2]}
    B=SetMaterialAttributes!(MaterialAttributes=@n4, BaseColor=@n12, EmissiveColor=@n20, Metallic=@n21) {attributeSetTypes [#1, #2, #3]}
    Alpha=@n22
   ))
  n23 = GetMaterialAttributes(None=FunctionInput() {inputName=Material, inputType=FunctionInput_MaterialAttributes}) {attributeGetTypes [#1, #2]}
  NamedRerouteDeclaration(None=@n23.EmissiveColor) {name=Emissive, nodeColor=<rgba 1 0 0.75 1>}
  NamedRerouteDeclaration(None=@n23.MaterialAttributes) {name=MA, nodeColor=<rgba 0.3406 1 0.0156 1>}
  n24 = NamedRerouteUsage()
  n25 = VectorParameter() {group=CVCrystal, parameterName=CV_SpellColor1, defaultValue=<rgba 0.6927 0.1078 0.6051 1>}
  n26 = NamedRerouteUsage()
  n27 = NamedRerouteUsage()
  n28 = VectorParameter() {group=CVCrystal, parameterName=CV_SpellColor2, sortPriority=33, defaultValue=<rgba 0.0358 0.3229 0 1>}
  n29 = StaticSwitchParameter(
   True=AppendVector(
    A=MaterialFunctionCall!(UV=@n24, UV Tilling, PannerValue) {materialFunction=/BRCosmetics/Weapons/FORT_Melee/Pickaxe_DyedDuelist/Materials/MF_RadialUVPanner}.UV
    B=Constant()
   )
   False=LinearInterpolate(A=Multiply(A=@n25.RGB, B=@n26), B=Multiply(A=@n27, B=@n28.RGB), Alpha=NamedRerouteUsage())
  ) {group=CVCrystal, parameterName=CV_UVPivotDebug, sortPriority=40}
  n30 = Multiply(A=NamedRerouteUsage())
  n31 = Constant() {r=0.5}
  n32 = Step(X=OneMinus(None=Distance(A=@n30, B=@n31)))
  NamedRerouteDeclaration(None=Multiply(A=Multiply(A=@n29, B=@n32)) {constB=2}) {name=CrystalEffect, nodeColor=<rgba 0.4766 1 0 1>}
  n33 = Multiply(A=Time()) {constB=0.5}
  n34 = Multiply(A=NamedRerouteUsage()) {constB=0.5}
  n35 = MaterialFunctionCall!(UV=NamedRerouteUsage(), UV Tilling, PannerValue=AppendVector(A=Constant(), B=Add(A=@n33, B=@n34))) {materialFunction=&/MF_RadialUVPanner}
  n36 = TextureSample!(UVs=Add(A=@n35.UV) {constB=-0.2}, Tex, "DDX(UVs)"=@n35.DDX, "DDY(UVs)"=@n35.DDX, Apply View MipBias) {
   mipValueMode=TMVM_Derivative
   samplerType=SAMPLERTYPE_LinearColor
   texture=/BRCosmetics/Weapons/FORT_Melee/Pickaxe_DyedDuelist/Textures/T_DyedDuelist_FX
  }
  NamedRerouteDeclaration(None=@n36.G) {name=SpellTexG, nodeColor=<rgba 0 1 0 1>}
  NamedRerouteDeclaration(None=@n36.R) {name=SpellTexR, nodeColor=<rgba 1 0.1172 0 1>}
  NamedRerouteDeclaration(None=FunctionInput() {inputName=Reactive, inputType=FunctionInput_Scalar, sortPriority=2}) {name=Reactive, nodeColor=<rgba 0.7814 0.2419 1 1>}
  NamedRerouteDeclaration(None=@n23.BaseColor) {name=BaseColor, nodeColor=<rgba 0.1661 0.3253 1 1>}
  n37 = TextureCoordinate()
  n38 = NamedRerouteDeclaration(None=@n37) {name=UVCoord, nodeColor=<rgba 0.5938 1 0 1>}
  n39 = ScalarParameter() {defaultValue=15, group=CVCrystal, parameterName=CV_InteriorUVTiling, sortPriority=4}
  n40 = ScalarParameter() {group=CVCrystal, parameterName=CV_Depth, sortPriority=6}
  n41 = ScalarParameter() {defaultValue=-10, group=CVCrystal, parameterName=CV_DepthMaximum, sortPriority=7}
  n42 = Multiply(A=@n37, B=ScalarParameter() {defaultValue=0.1, group=CVCrystal, parameterName=CV_DepthUVTiling, sortPriority=8})
  NamedRerouteDeclaration(None=BumpOffset(Coordinate=Multiply(A=@n38, B=@n39), Height=LinearInterpolate(A=@n40, B=@n41, Alpha=@n42) {constA=-4, constB=-20})) {name=UVOffsetCoord, nodeColor=<rgba 0.9453 0 1 1>}
  n43 = PixelNormalWS()
  n44 = CameraVectorWS()
  n45 = OneMinus(None=Saturate(None=DotProduct(A=@n43, B=@n44)))
  NamedRerouteDeclaration(None=Multiply(A=Multiply(A=@n45, B=@n45), B=ScalarParameter() {defaultValue=1, group=CVCrystal, parameterName=CV_FresnelIntensity, sortPriority=20})) {name=Fresnel, nodeColor=<rgba 0.1953 0 1 1>}
  n46 = ReflectionVectorWS()
  n47 = ScalarParameter() {defaultValue=0.1, group=CVCrystal, parameterName=CV_ReflectionRoughness, sortPriority=22}
  n48 = VectorParameter() {group=CVCrystal, parameterName=CV_ReflectionColor, sortPriority=21, defaultValue=<rgba 0.0952 0.0364 0.1146 1>}
  NamedRerouteDeclaration(None=Multiply(A=SkyLightEnvMapSample(Direction=@n46, Roughness=@n47), B=@n48.RGB)) {name=Reflection, nodeColor=<rgba 1 0.1406 0 1>}
  NamedRerouteDeclaration(None=FunctionInput() {inputName=ColorProgress, inputType=FunctionInput_Scalar, sortPriority=2}) {name=ColorProgress, nodeColor=<rgba 0 1 0.1328 1>}
  n49 = VectorParameter() {group=CVCrystal, parameterName=CV_PivotUVPos, sortPriority=35, defaultValue=<rgba 1.25 0 -14.5 0>}
  n50 = Divide(A=Add(A=PreSkinnedPosition(), B=ComponentMask(None=@n49.RGB) {b=true, g=true, r=true}), B=Constant3Vector() {constant=<rgba 2.6 1 3.5 0>}) {constB=4.65}
  NamedRerouteDeclaration(None=ComponentMask(None=VertexInterpolator(VS=@n50).PS) {b=true, r=true}) {name=PivotUV, nodeColor=<rgba 1 0 0.7266 1>}
  NamedRerouteDeclaration(None=FunctionInput() {inputName=SpellSpeed, inputType=FunctionInput_Scalar, sortPriority=3}) {name=SpellSpeed, nodeColor=<rgba 1 0.289 0.361 1>}
 ]}
