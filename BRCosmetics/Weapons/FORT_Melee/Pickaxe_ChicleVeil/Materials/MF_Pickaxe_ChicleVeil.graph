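# Material function for the ChicleVeil pickaxe that overrides BaseColor and EmissiveColor on incoming material attributes.
# Its masks come from TAMask, WeakPoint and Reactive inputs, FX texture channels, Fresnel and pre-skinned position gradients computed per vertex.
# A tinted Fresnel glow uses two opposite-scrolling fire noise samples, simplified at lower quality levels.
# An edge thin-film iridescence branch exists but does not connect to the output.
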MaterialFunction /BRCosmetics/Weapons/FORT_Melee/Pickaxe_ChicleVeil/Materials/MF_Pickaxe_ChicleVeil
graph {nodes [
  n1 = NamedRerouteUsage()
  n2 = Multiply(A=NamedRerouteUsage()) {constB=0.6}
  n3 = Multiply(A=Multiply(A=NamedRerouteUsage(), B=NamedRerouteUsage()), B=NamedRerouteUsage())
  n4 = NamedRerouteUsage()
  n5 = VectorParameter() {group=ChicleVeil, parameterName=ChicleVeil_FresnelColor, defaultValue=<rgba 0.2384 0.0194 0.3663 1>}
  n6 = Add(A=NamedRerouteUsage(), B=NamedRerouteUsage())
  n7 = NamedRerouteUsage()
  n8 = Multiply(A=Saturate(None=Add(A=@n6, B=@n7))) {constB=0.8}
  n9 = Add(A=Power(Base=NamedRerouteUsage()) {constExponent=1.5}, B=Multiply(A=OneMinus(None=@n8), B=NamedRerouteUsage()))
  n10 = Multiply(A=TextureCoordinate() {uTiling=0.8, vTiling=0.8}) {constB=2}
  n11 = Time()
  n12 = Multiply(A=Constant2Vector() {g=1}) {constB=0.1}
  n13 = TextureSample(UVs=Add(A=@n10, B=Multiply(A=@n11, B=@n12))) {samplerSource=SSM_Wrap_WorldGroupSettings, texture=/Game/Global/Textures/Default/Noise/T_FireNoise_01}
  n14 = Add(A=@n10, B=Multiply(A=@n11, B=Multiply(A=@n12) {constB=-1}))
  n15 = Max(
   A=@n13.R
   B=TextureSample(UVs=Multiply(A=@n14) {constB=1.1}) {samplerSource=SSM_Wrap_WorldGroupSettings, texture=&T_FireNoise_01}.R
  )
  n16 = Multiply(A=QualitySwitch(Default=@n15, Low=Constant() {r=1}, High=@n15, Medium=@n13.R, Epic=@n15)) {constB=2}
  n17 = Multiply(A=ScalarParameter() {defaultValue=1, group=ChicleVeil, parameterName=ChicleVeil_FresnelIntensity}, B=Saturate(None=Multiply(A=@n9, B=@n16)))
  n18 = Multiply(A=NamedRerouteUsage(), B=LinearInterpolate(A=Multiply(A=@n5.RGB, B=@n17), B=NamedRerouteUsage(), Alpha=NamedRerouteUsage()))
  FunctionOutput(None=SetMaterialAttributes!(MaterialAttributes=NamedRerouteUsage(), BaseColor=LinearInterpolate(A=@n1, B=@n2, Alpha=@n3), EmissiveColor=Multiply(A=@n4, B=@n18)) {attributeSetTypes [#1, #2]})
  n19 = NamedRerouteUsage()
  n20 = ScalarParameter() {defaultValue=1, group=ThinFilm ChicleVeilEdge, parameterName=ThinFilmScaleEdge, sortPriority=5}
  n21 = Fresnel() {baseReflectFraction=0.08, exponent=1}
  n22 = Power(Base=@n21, Exp=ScalarParameter() {defaultValue=0.5, group=ThinFilm ChicleVeilEdge, parameterName=ThinFilmWarpEdge, sortPriority=4})
  n23 = TextureSampleParameter2D(UVs=Add(A=Multiply(A=@n20, B=@n22), B=ScalarParameter() {group=ThinFilm ChicleVeilEdge, parameterName=ThinFilmOffsetEdge})) {
   group=ThinFilm ChicleVeilEdge
   parameterName=Thin Film Texture Edge
   sortPriority=1
   texture=/Game/Global/Textures/Default/Thinfilm/T_ThinFilm_Spectrum_COLOR
  }
  n24 = ScalarParameter() {defaultValue=0.5, group=ThinFilm ChicleVeilEdge, parameterName=ThinFilmIntensityEdge, sortPriority=2}
  n25 = Power(Base=@n21, Exp=ScalarParameter() {defaultValue=2, group=ThinFilm ChicleVeilEdge, parameterName=ThinFilmExponentEdge, sortPriority=3})
  n26 = Multiply(A=LinearInterpolate(Alpha=NamedRerouteUsage()) {constA=1, constB=0.01}, B=@n25)
  n27 = ScalarParameter() {group=ThinFilm ChicleVeilEdge, parameterName=RoughnessInfluenceEdge}
  LinearInterpolate(A=@n19, B=Add(A=@n19, B=Multiply(A=@n23.RGB, B=@n24)), Alpha=Saturate(None=LinearInterpolate(A=@n25, B=@n26, Alpha=@n27)))
  n28 = GetMaterialAttributes(None=FunctionInput() {inputType=FunctionInput_MaterialAttributes}) {attributeGetTypes [
    #1
    #3
    #4
    #2
   ]}
  NamedRerouteDeclaration(None=@n28.BaseColor) {name=BaseColor, nodeColor=<rgba 0 0.5547 1 1>}
  NamedRerouteDeclaration(None=@n28.MaterialAttributes) {name=MA, nodeColor=<rgba 0.0078 0 1 1>}
  n29 = TextureSample() {texture=/BRCosmetics/Weapons/FORT_Melee/Pickaxe_ChicleVeil/Textures/T_ChicleVeil_Pickaxe_FX2}
  n30 = Step(X=@n29.B) {constX=0.1, constY=0.1}
  n31 = Add(A=PreSkinnedPosition(), B=Constant3Vector() {constant=<rgba 0 0 -17.5 0>})
  n32 = VertexInterpolator(VS=Divide(A=@n31, B=Constant3Vector() {constant=<rgba 6.9 1 61.5 0>}))
  n33 = ComponentMask(None=@n32.PS) {r=true}
  n34 = VertexInterpolator(VS=Divide(A=@n31, B=Constant3Vector() {constant=<rgba 6.9 1 52 0>}))
  n35 = ComponentMask(None=@n34.PS) {r=true}
  n36 = ComponentMask(None=@n34.PS) {b=true}
  n37 = Add(A=Multiply(A=Step(Y=@n33) {constX=0}, B=ComponentMask(None=@n32.PS) {b=true}), B=Multiply(A=Step(X=@n35), B=@n36))
  NamedRerouteDeclaration(None=Multiply(A=NamedRerouteDeclaration(None=@n30) {name=SwordEdgeMask, nodeColor=<rgba 0 0.625 1 1>}, B=OneMinus(None=@n37))) {name=SwordMask, nodeColor=<rgba 0 0.8125 1 1>}
  n38 = NamedRerouteUsage()
  n39 = LinearInterpolate(
   A=Constant3Vector() {constant=<rgba 0.482 0.146 0.54 1>}
   B=Constant3Vector() {constant=<rgba 0.2602 0.001 0.27 1>}
   Alpha=SmoothStep(Value=@n38) {constMax=0.5}
  )
  n40 = Constant3Vector() {constant=<rgba 0.2588 0.0198 0.4 1>}
  n41 = SmoothStep(Value=@n38) {constMin=0.5}
  NamedRerouteDeclaration(None=Multiply(A=LinearInterpolate(A=@n39, B=@n40, Alpha=@n41), B=NamedRerouteUsage())) {name=OutEdgeColor, nodeColor=<rgba 1 0 0.4922 1>}
  NamedRerouteDeclaration(None=@n29.G) {name=InnerEdgeMask, nodeColor=<rgba 0 0.4141 1 1>}
  n42 = NamedRerouteUsage()
  n43 = LinearInterpolate(
   A=Constant3Vector() {constant=<rgba 0.491 0.5149 0.7682 1>}
   B=Constant3Vector() {constant=<rgba 0.1529 0.3663 0.3419 1>}
   Alpha=SmoothStep(Value=@n42) {constMax=0.5}
  )
  n44 = Constant3Vector() {constant=<rgba 0.1095 0.2232 0.3185 1>}
  n45 = SmoothStep(Value=@n42) {constMin=0.5}
  NamedRerouteDeclaration(None=LinearInterpolate(A=NamedRerouteUsage(), B=LinearInterpolate(A=@n43, B=@n44, Alpha=@n45), Alpha=NamedRerouteUsage())) {name=EdgeColor, nodeColor=<rgba 0.2422 1 0 1>}
  NamedRerouteDeclaration(None=@n28.Roughness) {name=Roughness, nodeColor=<rgba 0 0.0391 1 1>}
  NamedRerouteDeclaration(None=Fresnel() {baseReflectFraction=0.2, exponent=2}) {name=Fresnel, nodeColor=<rgba 0 1 0.8828 1>}
  NamedRerouteDeclaration(None=FunctionInput() {inputName=WeakPoint, inputType=FunctionInput_Scalar, sortPriority=3}) {name=Weakpoints, nodeColor=<rgba 0.0781 1 0 1>}
  NamedRerouteDeclaration(None=@n28.Metallic) {name=Metallic, nodeColor=<rgba 1 0.2812 0 1>}
  n46 = NamedRerouteDeclaration(None=FunctionInput() {inputName=TAMask, sortPriority=1}) {name=TAMask, nodeColor=<rgba 0 1 0.6719 1>}
  NamedRerouteDeclaration(None=ComponentMask(None=@n46) {r=true}) {name=CrystalMask, nodeColor=<rgba 0 1 0.5547 1>}
  NamedRerouteDeclaration(None=ComponentMask(None=@n46) {g=true}) {name=MetalHandle, nodeColor=<rgba 1 0.5859 0 1>}
  NamedRerouteDeclaration(None=@n29.R) {name=FresnelBlade, nodeColor=<rgba 1 0 0.6797 1>}
  NamedRerouteDeclaration(None=Step(X=@n36)) {name=SwordBlade, nodeColor=<rgba 1 0 0.2578 1>}
  NamedRerouteDeclaration(None=FunctionInput() {inputName=Reactive, inputType=FunctionInput_Scalar, sortPriority=3}) {name=Reactive, nodeColor=<rgba 1 0 0.971 1>}
  NamedRerouteDeclaration(None=@n28.EmissiveColor) {name=Emissive, nodeColor=<rgba 0 1 0.4375 1>}
 ]}
