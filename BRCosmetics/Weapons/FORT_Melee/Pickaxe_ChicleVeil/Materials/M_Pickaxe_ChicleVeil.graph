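# Master material for the ChicleVeil melee pickaxe, supporting skeletal meshes, Nanite and multiple shading models via material attributes.
# The shared 2023 character PreFX function feeds a crystal effect, masked by the Tech Art Mask's red channel, then a pickaxe-specific function and shared PostFX.
# Reactive and VibrantEmissiveAmount parameters drive both custom functions, the latter as crystal color progress and pickaxe weak point.
# SpellSpeed feeds only the crystal effect.
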Material /BRCosmetics/Weapons/FORT_Melee/Pickaxe_ChicleVeil/Materials/M_Pickaxe_ChicleVeil
graph {
 properties {
  bUseMaterialAttributes=true
  bUsedWithMeshDeformer=true
  bUsedWithNanite=true
  bUsedWithSkeletalMesh=true
  pixelDepthOffsetMode=PDOM_Legacy
  shadingModel=MSM_FromMaterialExpression
  usedShadingModels=Default Lit | Subsurface | Clear Coat | Subsurface Profile | Cloth
 }
 nodes [
  n1 = MaterialFunctionCall() {materialFunction=&MF_CharacterParent_2023_PreFX}
  n2 = ComponentMask(None=@n1.Tech Art Mask) {r=true}
  n3 = ScalarParameter() {parameterName=Reactive}
  n4 = ScalarParameter() {parameterName=VibrantEmissiveAmount}
  n5 = ScalarParameter() {parameterName=SpellSpeed}
  MaterialFunctionCall(
   Material=MaterialFunctionCall!(
    In=MaterialFunctionCall!(Material=@n1.Result, Alpha=@n2, Reactive=@n3, ColorProgress=@n4, SpellSpeed=@n5) {materialFunction=&/MF_ChicleVeil_Crystal}.Result
    TAMask=@n1.Tech Art Mask
    Reactive=@n3
    WeakPoint=@n4
   ) {materialFunction=&/MF_Pickaxe_ChicleVeil}.Result
   M=@n1.M
   Pre Skinned Local Position=@n1.M
   Distance Blend=@n1.M
   Fresnel=@n1.M
  ) {materialFunction=&MF_CharacterParent_2023_PostFX}
 ]
}
