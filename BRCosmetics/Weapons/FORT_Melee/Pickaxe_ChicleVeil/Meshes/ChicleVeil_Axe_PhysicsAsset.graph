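# Physics asset for the ChicleVeil two-handed pickaxe, a melee weapon cosmetic mesh.
# It has a single capsule body on the Melee_TwoHanded_Handle bone, offset along Z and rotated about ninety degrees to run along the handle.
# No constraints are defined, so the weapon simulates as one rigid shape.
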PhysicsAsset /BRCosmetics/Weapons/FORT_Melee/Pickaxe_ChicleVeil/Meshes/ChicleVeil_Axe_PhysicsAsset
bodies {Melee_TwoHanded_Handle [{shapeType=Capsule, length=52.5165, radius=14.3719, center=<-0.118 0.9185 26.5463>, rotation=<rot 0.862 -90 -0.031>}]}
constraints []
